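# Material function for the Vertigo character body that adds animated throbbing to input material attributes through panning Perlin noise.
# World-aligned noise normals, sampled on a noise-warped pre-skinned local position and flattened by intensity, blend into the base normal.
# Biased noise scaled by a thickness map pushes vertices along their normals.
# Static switches toggle both effects, Alpha blends the result, and a "Water" output exposes the noise normal.
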MaterialFunction /BRCosmetics/Characters/Player/Male/Large/Bodies/M_LRG_Vertigo/Materials/MF_Throbbing
graph {nodes [
  n1 = FunctionInput() {inputType=FunctionInput_MaterialAttributes}
  n2 = GetMaterialAttributes(None=@n1) {attributeGetTypes [#1, #2]}
  n3 = TextureObjectParameter() {
   parameterName=Panning Texture
   samplerType=SAMPLERTYPE_Normal
   texture=/Game/Characters/Player/Male/Large/Bodies/M_LRG_Clash_V/Textures/Simple_Perlin_N
  }
  n4 = VectorParameter() {group=Throbbing, parameterName=Throbbing Scale, defaultValue=<rgba 180 180 180 1>}
  n5 = ScalarParameter() {defaultValue=1, group=Throbbing, parameterName=TriPlanner_AxisBlendStrength}
  n6 = TextureObjectParameter() {
   parameterName=Panning Texture
   samplerType=SAMPLERTYPE_Masks
   texture=/Game/Characters/Player/Male/Large/Bodies/M_LRG_Clash_V/Textures/Simple_Perlin_M
  }
  n7 = Constant() {r=30}
  n8 = FunctionInput() {inputName=Pre SKinned Local}
  n9 = ScalarParameter() {defaultValue=-0.2, parameterName=Pann}
  n10 = Multiply(
   A=ComponentMask(None=MaterialFunctionCall!(TextureObject=@n6, TextureSize=@n7, WorldPosition=@n8, Export Float 4, In=@n9, World Space Normal, ProjectionTransitionContrast) {materialFunction=/Game/Athena/SafeZone/WorldAlignedTexture_ScalePan}.XYZ Texture) {g=true}
   B=ScalarParameter() {defaultValue=0.3, parameterName=warp Amount}
  )
  n11 = VectorParameter() {group=Throbbing, parameterName=TriPlanner Speed, defaultValue=<rgba 0 0.2 0.2 1>}
  n12 = AppendVector(A=@n11.R, B=@n11.R)
  n13 = MaterialFunctionCall(
   Normal=MaterialFunctionCall!(
    TextureObject=@n3
    TextureSize=@n4.RGB
    Normal
    WorldSpace
    ProjectionTransitionContrast=@n5
    WorldPosition=Add(A=@n10, B=@n8)
    Use High Quality Normals
    Pan_YZ=@n12
    Pan_XZ=@n12
    Pan_XY=@n12
   ) {materialFunction=/Game/Weapons/WeaponSkins/Wraps/MaterialFunctions/WorldAlignedNormal_ComplexPan}.XYZ Texture
   Flatness=LinearInterpolate(Alpha=ScalarParameter() {group=Throbbing, parameterName=Throbbing_Intensity}) {constA=1, constB=0}
  ) {materialFunction=&FlattenNormal}
  n14 = StaticSwitchParameter(
   True=MaterialFunctionCall(BaseNormal=@n2.Normal, AdditionalNormal=@n13.Result) {materialFunction=&BlendAngleCorrectedNormals}.Result
   False=@n2.Normal
  ) {defaultValue=true, parameterName=use Normal Map}
  n15 = TextureObjectParameter() {
   parameterName=Panning Texture
   samplerType=SAMPLERTYPE_Masks
   texture=&Simple_Perlin_M
  }
  n16 = ComponentMask(None=MaterialFunctionCall(TextureObject=@n15, TextureSize=@n4.RGB, WorldPosition=@n8, In=@n12, ProjectionTransitionContrast=@n5) {materialFunction=&WorldAlignedTexture_ScalePan}.XYZ Texture) {r=true}
  n17 = Multiply(
   A=Multiply(
    A=ConstantBiasScale(None=@n16) {bias=-0.5, scale=1}
    B=ScalarParameter() {defaultValue=4, group=Throbbing, parameterName=VertexThrobbingIntensity}
   )
   B=FunctionInput(Preview=Constant() {r=1}) {bUsePreviewValueAsDefault=true, inputName=ThicknessMap}
  )
  n18 = VertexNormalWS()
  n19 = StaticSwitchParameter(True=Add(A=@n2.WorldPositionOffset, B=Multiply(A=@n17, B=@n18)), False=@n2.WorldPositionOffset) {group=Throbbing, parameterName=useWPOthrobbing}
  FunctionOutput(None=BlendMaterialAttributes(
    A=@n1
    B=SetMaterialAttributes!(MaterialAttributes=@n2.MaterialAttributes, Normal=@n14, WorldPositionOffset=@n19) {attributeSetTypes [#1, #2]}
    Alpha=FunctionInput() {inputName=Alpha, inputType=FunctionInput_Scalar}
   ))
  FunctionOutput(None=@n13.Result) {outputName=Water}
 ]}
