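# Parent material for the SoupGoal female character body, layering switchable effects between the shared character PreFX and PostFX functions.
# Inner parallax blends bump-offset, panning noise projected biplanarly from pre-skinned position into base color, masked by base-color brightness.
# Panning vein noise tints color, alters normals and feeds Vertigo's throbbing function.
# Reactivity sums five start-time-driven hit envelopes with crackle noise, revealing glowing veins through a gradient mask.
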Material /BRCosmetics/Characters/Player/Female/Medium/Bodies/F_MED_SoupGoal/Materials/M_SoupGoal_Parent
graph {
 properties {
  bHasPixelAnimation=true
  bUseMaterialAttributes=true
  bUsedWithMeshDeformer=true
  bUsedWithMorphTargets=true
  bUsedWithNanite=true
  bUsedWithNiagaraMeshParticles=true
  bUsedWithSkeletalMesh=true
  pixelDepthOffsetMode=PDOM_Legacy
  shadingModel=MSM_FromMaterialExpression
  usedShadingModels=Default Lit | Subsurface | Clear Coat | Subsurface Profile | Cloth
 }
 nodes [
  n1 = GetMaterialAttributes(None=NamedRerouteUsage()) {attributeGetTypes [#1, #2, #3]}
  n2 = SphereMask(
   A=Desaturation(None=@n1.Base Color)
   B=ScalarParameter() {group=SoupGoal, parameterName=InnerParallax_BCMask_Center, sliderMax=1}
   Radius=ScalarParameter() {defaultValue=0.045, group=SoupGoal, parameterName=InnerParallax_BCMask_Range, sliderMax=1}
   Hardness=ScalarParameter() {defaultValue=0.5, group=SoupGoal, parameterName=InnerParallax_BCMask_Hardness, sliderMax=1}
  ) {attenuationRadius=0.4, hardnessPercent=50}
  n3 = NamedRerouteUsage()
  n4 = MaterialFunctionCall(X=Constant() {r=-0.0183}, Y=ScalarParameter() {defaultValue=0.0168, group=SoupGoal, parameterName=InnerParallax_VPanSpeed}) {materialFunction=&MakeFloat2}
  n5 = Multiply(A=AppendVector(A=NamedRerouteUsage(), B=Constant() {r=-1}), B=NamedRerouteUsage()) {constA=1}
  n6 = MaterialFunctionCall!(UVs=NamedRerouteUsage(), Normal=Constant3Vector() {constant=<rgba 0 0 1 0>}) {materialFunction=/Engine/Functions/Engine_MaterialFunctions01/Texturing/DeriveTangentBasis}
  n7 = CrossProduct(A=VertexNormalWS(), B=CameraVectorWS())
  n8 = AppendVector(A=DotProduct(A=@n6.Binormal, B=@n7), B=Multiply(A=NamedRerouteUsage(), B=DotProduct(A=@n6.Tangent, B=@n7)))
  n9 = NamedRerouteUsage()
  n10 = Panner(
   Coordinate=@n5
   Time=NamedRerouteUsage()
   Speed=MaterialFunctionCall(Y=ScalarParameter() {defaultValue=0.01, group=SoupGoal, parameterName=InnerParallax_HeightVPanSpeed}) {materialFunction=&MakeFloat2}.Result
  ) {speedY=-0.1}
  n11 = TextureSampleParameter2D(UVs=Multiply(
    A=StaticSwitchParameter(True=@n5, False=@n10) {group=SoupGoal, parameterName=InnerParallax_DisableAnimation}
    B=ScalarParameter() {defaultValue=11, group=SoupGoal, parameterName=InnerParallax_HeightmapScale}
   )) {
   group=SoupGoal
   parameterName=InnerParallax_Heightmap
   samplerType=SAMPLERTYPE_Grayscale
   texture=&LowResBlurredNoise
  }
  n12 = Subtract(A=@n11.R, B=ScalarParameter() {defaultValue=1, group=SoupGoal, parameterName=InnerParallax_BumpOffsetRefHeight})
  n13 = ScalarParameter() {defaultValue=0.15, group=SoupGoal, parameterName=InnerParallax_BumpOffsetRatio}
  n14 = Subtract(A=@n5, B=Multiply(A=Multiply(A=@n8, B=@n9), B=Multiply(A=@n12, B=@n13)))
  n15 = TextureSampleParameter2D(UVs=StaticSwitchParameter(True=@n14, False=Panner(Coordinate=@n14, Time=@n3, Speed=@n4.Result) {speedY=-0.15}) {group=SoupGoal, parameterName=InnerParallax_DisableAnimation}) {
   group=SoupGoal
   parameterName=InnerParallax_Pattern
   texture=/BRCosmetics/Characters/Player/Female/Medium/Bodies/F_MED_SoupGoal/Textures/T_SoupGoal_InteriorNoise_A
  }
  n16 = NamedRerouteUsage()
  n17 = Multiply(A=AppendVector(A=NamedRerouteUsage(), B=Constant() {r=-1}), B=NamedRerouteUsage()) {constA=1}
  n18 = MaterialFunctionCall(UVs=NamedRerouteUsage(), Normal=Constant3Vector() {constant=<rgba 0 0 1 0>}) {materialFunction=&DeriveTangentBasis}
  n19 = CrossProduct(A=VertexNormalWS(), B=CameraVectorWS())
  n20 = AppendVector(A=DotProduct(A=@n18.Binormal, B=@n19), B=Multiply(A=NamedRerouteUsage(), B=DotProduct(A=@n18.Tangent, B=@n19)))
  n21 = NamedRerouteUsage()
  n22 = Panner(
   Coordinate=@n17
   Time=NamedRerouteUsage()
   Speed=MaterialFunctionCall(Y=ScalarParameter() {defaultValue=0.01, group=SoupGoal, parameterName=InnerParallax_HeightVPanSpeed}) {materialFunction=&MakeFloat2}.Result
  ) {speedY=-0.1}
  n23 = TextureSampleParameter2D(UVs=Multiply(
    A=StaticSwitchParameter(True=@n17, False=@n22) {group=SoupGoal, parameterName=InnerParallax_DisableAnimation}
    B=ScalarParameter() {defaultValue=11, group=SoupGoal, parameterName=InnerParallax_HeightmapScale}
   )) {
   group=SoupGoal
   parameterName=InnerParallax_Heightmap
   samplerType=SAMPLERTYPE_Grayscale
   texture=&LowResBlurredNoise
  }
  n24 = Subtract(A=@n23.R, B=ScalarParameter() {defaultValue=1, group=SoupGoal, parameterName=InnerParallax_BumpOffsetRefHeight})
  n25 = ScalarParameter() {defaultValue=0.15, group=SoupGoal, parameterName=InnerParallax_BumpOffsetRatio}
  n26 = Subtract(A=@n17, B=Multiply(A=Multiply(A=@n20, B=@n21), B=Multiply(A=@n24, B=@n25)))
  n27 = TextureSampleParameter2D(UVs=StaticSwitchParameter(True=@n26, False=Panner(Coordinate=@n26, Time=@n16, Speed=@n4.Result) {speedY=-0.15}) {group=SoupGoal, parameterName=InnerParallax_DisableAnimation}) {
   group=SoupGoal
   parameterName=InnerParallax_Pattern
   texture=&T_SoupGoal_InteriorNoise_A
  }
  n28 = VertexInterpolator(VS=PreSkinnedNormal())
  n29 = Multiply(A=@n28.PS, B=Constant3Vector() {constant=<rgba 1 1 0 0>})
  n30 = DotProduct(A=Abs(None=Normalize(VectorInput=@n29)), B=Constant3Vector() {constant=<rgba 0 1 0 0>})
  n31 = MaterialFunctionCall(Multiplier=Constant() {r=0.5}) {materialFunction=&Pi}
  n32 = ScalarParameter() {defaultValue=0.2, group=SoupGoal, parameterName=InnerParallax__BiplanarTransitionWidth, sliderMax=1}
  n33 = Subtract(
   A=Divide(A=ArccosineFast!(None=@n30), B=@n31.Result)
   B=Subtract(
    A=ScalarParameter() {defaultValue=0.5, group=SoupGoal, parameterName=InnerParallax_BiplanarBalance, sliderMax=1}
    B=Multiply(A=@n32) {constB=0.5}
   )
  )
  n34 = NamedRerouteDeclaration(None=Saturate(None=Divide(A=@n33, B=@n32)) {desc=Blend - Front/Side}) {name=WorldUV_XYBlend, nodeColor=<rgba 0.8516 0 1 1>}
  n35 = LinearInterpolate(A=@n15.R, B=@n27.R, Alpha=@n34)
  n36 = NamedRerouteUsage()
  n37 = MaterialFunctionCall(X=Constant() {r=0.0179}, Y=ScalarParameter() {defaultValue=0.0193, group=SoupGoal, parameterName=InnerParallax_VPanSpeed2}) {materialFunction=&MakeFloat2}
  n38 = TextureSampleParameter2D(UVs=Multiply(A=Panner(Coordinate=@n14, Time=@n36, Speed=@n37.Result) {speedY=-0.15}) {constB=1.3728}) {
   group=SoupGoal
   parameterName=InnerParallax_Pattern
   texture=&T_SoupGoal_InteriorNoise_A
  }
  n39 = NamedRerouteUsage()
  n40 = TextureSampleParameter2D(UVs=Multiply(A=Panner(Coordinate=@n26, Time=@n39, Speed=@n37.Result) {speedY=-0.15}) {constB=1.3728}) {
   group=SoupGoal
   parameterName=InnerParallax_Pattern
   texture=&T_SoupGoal_InteriorNoise_A
  }
  n41 = LinearInterpolate(A=@n35, B=LinearInterpolate(A=@n38.R, B=@n40.R, Alpha=@n34))
  n42 = DotProduct(A=Abs(None=@n28.PS), B=Constant3Vector() {constant=<rgba 0 0 1 0>})
  n43 = OneMinus(None=Divide(A=ArccosineFast(None=@n42), B=@n31.Result))
  n44 = ScalarParameter() {defaultValue=0.5, group=SoupGoal, parameterName=InnerParallax_BiplanarTopTreshold, sliderMax=1}
  n45 = Saturate(None=Multiply(
    A=Subtract(A=@n43, B=@n44)
    B=ScalarParameter() {defaultValue=5, group=SoupGoal, parameterName=InnerParallax_BiplanarTopContrast, sliderMax=1}
   )) {desc=Blend - Vertical}
  n46 = Multiply(
   A=LinearInterpolate(
    A=VectorParameter() {group=SoupGoal, parameterName=InnerParallaxColor_Low, defaultValue=<rgba 0.0285 0.0017 0.0469 1>}.RGB
    B=VectorParameter() {group=SoupGoal, parameterName=InnerParallaxColor_High, defaultValue=<rgba 0.3698 0 0.5885 1>}.RGB
    Alpha=StaticSwitchParameter(True=@n35, False=@n41) {group=SoupGoal, parameterName=InnerParallax_DisableAnimation}
   )
   B=SmoothStep(Value=OneMinus(None=@n45))
  )
  n47 = ScalarParameter() {defaultValue=0.25, group=SoupGoal, parameterName=InnerParallax_Brightness}
  Add(A=@n1.Emissive Color, B=Multiply(A=@n2, B=Multiply(A=@n46, B=@n47))) {constB=0}
  n48 = MaterialFunctionCall() {materialFunction=&MF_CharacterParent_2023_PreFX}
  n49 = Saturate(None=DotProduct(A=VectorParameter() {group=Throbbing, parameterName=ThrobbingChannelMask, defaultValue=<rgba 0 0 1 0>}.RGB, B=@n48.Tech Art Mask))
  n50 = OneMinus(None=@n49)
  n51 = NamedRerouteUsage()
  n52 = Add(
   A=@n51
   B=RotateAboutAxis(NormalizedRotationAxis=Constant3Vector() {constant=<rgba 0 0 1 0>}, RotationAngle=Constant() {r=0.125}, PivotPoint=Constant(), Position=@n51)
  )
  n53 = StaticSwitchParameter(True=NamedRerouteUsage(), False=NamedRerouteUsage()) {group=SoupGoal, parameterName=EnableInnerParallax}
  n54 = GetMaterialAttributes(None=@n53) {attributeGetTypes [#3, #1]}
  n55 = AppendVector(A=Constant() {r=0.03}, B=Constant() {r=1})
  n56 = Multiply(A=Time(), B=ScalarParameter() {defaultValue=0.03, group=SoupGoal, parameterName=ThrobbingPanSpeed})
  n57 = TextureSampleParameter2D() {
   group=SoupGoal
   parameterName=Reactivity_GradientMask
   samplerSource=SSM_Wrap_WorldGroupSettings
   samplerType=SAMPLERTYPE_LinearColor
   texture=/BRCosmetics/Characters/Player/Female/Medium/Bodies/F_MED_SoupGoal/Textures/T_SoupGoal_Body_TechArtFX_Mask
  }
  n58 = PreSkinnedLocalBounds()
  n59 = Divide(A=Subtract(A=NamedRerouteUsage(), B=@n58.Min), B=@n58.Extents)
  n60 = ScalarParameter() {defaultValue=0.8, group=SoupGoal, parameterName=ThrobbingAnimExponent}
  n61 = Multiply(
   A=ScalarParameter() {defaultValue=1, group=SoupGoal, parameterName=ThrobbingScale}
   B=Power(Base=ComponentMask(None=@n59) {b=true}, Exp=@n60) {constExponent=4}
  )
  n62 = TextureSampleParameter2D(UVs=Add(A=Multiply(A=@n55, B=@n56), B=AppendVector(A=@n57.R, B=@n61))) {
   group=SoupGoal
   parameterName=ThrobbingNoise
   samplerSource=SSM_Wrap_WorldGroupSettings
   samplerType=SAMPLERTYPE_Masks
   texture=/BRCosmetics/Characters/Player/Female/Medium/Bodies/F_MED_SoupGoal/Textures/T_SoupGoal_VeinsAnimation_A
  }
  n63 = Multiply(A=ScalarParameter() {defaultValue=5, group=SoupGoal, parameterName=ThrobbingNormalIntensity}, B=@n62.RGB)
  n64 = VectorParameter() {group=SoupGoal, parameterName=Veins_BaseColor_FromNoise, defaultValue=<rgba 0.015 0.015 0.015 1>}
  n65 = Multiply(A=@n54.Base Color, B=@n64.RGB)
  n66 = ScalarParameter() {defaultValue=0.5, group=SoupGoal, parameterName=Veins_BaseColor_TintVsSolid}
  n67 = MaterialFunctionCall(Float3=@n62.RGB) {materialFunction=&BreakOutFloat3Components}
  n68 = Multiply(A=@n67.R, B=@n67.R)
  n69 = Multiply(A=Power(Base=Abs(None=@n68)) {constExponent=3}) {constB=100}
  n70 = SetMaterialAttributes!(
   MaterialAttributes=@n54.MaterialAttributes
   Normal=LinearInterpolate(A=Constant3Vector() {constant=<rgba 0 0 1 0>}, B=@n54.Normal, Alpha=Add(A=@n63))
   Base Color=LinearInterpolate(A=@n54.Base Color, B=LinearInterpolate(A=@n65, B=@n64.RGB, Alpha=@n66), Alpha=Saturate(None=@n69)) {constB=0}
  ) {attributeSetTypes [#3, #1]}
  n71 = NamedRerouteUsage()
  n72 = StaticSwitchParameter(
   True=MaterialFunctionCall!(
    Alpha=StaticSwitchParameter(True=@n50, False=@n49) {group=Throbbing, parameterName=ThrobbingMaskChannel_Invert}
    Pre SKinned Local=StaticSwitchParameter(True=@n52, False=@n51) {group=Throbbing, parameterName=Throbbing_BackpackRotationFix}
    ThicknessMap
    In=BlendMaterialAttributes(A=@n54.MaterialAttributes, B=@n70, Alpha=@n71) {vertexAttributeBlendType=UseA}
   ) {materialFunction=/BRCosmetics/Characters/Player/Male/Large/Bodies/M_LRG_Vertigo/Materials/MF_Throbbing}.Result
   False=@n53
  ) {defaultValue=true, group=SoupGoal, parameterName=EnableThrobbingVeins}
  n73 = GetMaterialAttributes(None=@n72) {attributeGetTypes [#2, #1]}
  n74 = VectorParameter() {group=SoupGoal_Reactivity, parameterName=ReactiveColor, defaultValue=<rgba 0 0.317 1 1>}
  n75 = ScalarParameter() {defaultValue=0.3, parameterName=Reactivity_AlternateVeinsMask_Exp}
  n76 = NamedRerouteUsage()
  n77 = LinearInterpolate(
   A=@n76
   B=@n57.B
   Alpha=ScalarParameter() {defaultValue=0.9, group=SoupGoal_Reactivity, parameterName=Reactivity_VeinsMask_Balance, sliderMax=1}
  )
  n78 = Subtract(A=Time(), B=ScalarParameter() {defaultValue=-100, group=SoupGoal_Reactivity, parameterName=ReactivityStartTime_0})
  n79 = ScalarParameter() {defaultValue=0.2, group=SoupGoal_Reactivity, parameterName=ReactivityRampUpDuration}
  n80 = Power(
   Base=Saturate(None=Divide(A=@n78, B=@n79) {constB=0.8})
   Exp=ScalarParameter() {defaultValue=0.8, group=SoupGoal_Reactivity, parameterName=ReactivityRampUpExp}
  ) {constExponent=0.8}
  n81 = Divide(A=@n78, B=ScalarParameter() {defaultValue=5, group=SoupGoal_Reactivity, parameterName=ReactivityRampDownDuration}) {constB=0.8}
  n82 = Power(
   Base=OneMinus(None=Saturate(None=@n81))
   Exp=ScalarParameter() {defaultValue=0.4, group=SoupGoal_Reactivity, parameterName=ReactivityRampDownExp}
  ) {constExponent=0.4}
  n83 = Subtract(A=Time(), B=ScalarParameter() {defaultValue=-100, group=SoupGoal_Reactivity, parameterName=ReactivityStartTime_1})
  n84 = Divide(A=@n83, B=ScalarParameter() {defaultValue=0.2, group=SoupGoal_Reactivity, parameterName=ReactivityRampUpDuration}) {constB=0.8}
  n85 = Saturate(None=Divide(A=@n83, B=ScalarParameter() {defaultValue=5, group=SoupGoal_Reactivity, parameterName=ReactivityRampDownDuration}) {constB=0.8})
  n86 = Multiply(
   A=Power(Base=Saturate(None=@n84), Exp=ScalarParameter() {defaultValue=0.8, group=SoupGoal_Reactivity, parameterName=ReactivityRampUpExp}) {constExponent=0.8}
   B=Power(Base=OneMinus(None=@n85), Exp=ScalarParameter() {defaultValue=0.4, group=SoupGoal_Reactivity, parameterName=ReactivityRampDownExp}) {constExponent=0.4}
  )
  n87 = ScalarParameter() {defaultValue=0.25, group=SoupGoal_Reactivity, parameterName=ReactivitySingleHitStrength}
  n88 = Subtract(A=Time(), B=ScalarParameter() {defaultValue=-100, group=SoupGoal_Reactivity, parameterName=ReactivityStartTime_2})
  n89 = Saturate(None=Divide(A=@n88, B=ScalarParameter() {defaultValue=0.2, group=SoupGoal_Reactivity, parameterName=ReactivityRampUpDuration}) {constB=0.8})
  n90 = ScalarParameter() {defaultValue=0.8, group=SoupGoal_Reactivity, parameterName=ReactivityRampUpExp}
  n91 = ScalarParameter() {defaultValue=5, group=SoupGoal_Reactivity, parameterName=ReactivityRampDownDuration}
  n92 = OneMinus(None=Saturate(None=Divide(A=@n88, B=@n91) {constB=0.8}))
  n93 = ScalarParameter() {defaultValue=0.4, group=SoupGoal_Reactivity, parameterName=ReactivityRampDownExp}
  n94 = Multiply(
   A=Multiply(A=Power(Base=@n89, Exp=@n90) {constExponent=0.8}, B=Power(Base=@n92, Exp=@n93) {constExponent=0.4})
   B=ScalarParameter() {defaultValue=0.25, group=SoupGoal_Reactivity, parameterName=ReactivitySingleHitStrength}
  )
  n95 = Subtract(A=Time(), B=ScalarParameter() {defaultValue=-100, group=SoupGoal_Reactivity, parameterName=ReactivityStartTime_3})
  n96 = ScalarParameter() {defaultValue=0.2, group=SoupGoal_Reactivity, parameterName=ReactivityRampUpDuration}
  n97 = Power(
   Base=Saturate(None=Divide(A=@n95, B=@n96) {constB=0.8})
   Exp=ScalarParameter() {defaultValue=0.8, group=SoupGoal_Reactivity, parameterName=ReactivityRampUpExp}
  ) {constExponent=0.8}
  n98 = Divide(A=@n95, B=ScalarParameter() {defaultValue=5, group=SoupGoal_Reactivity, parameterName=ReactivityRampDownDuration}) {constB=0.8}
  n99 = Power(
   Base=OneMinus(None=Saturate(None=@n98))
   Exp=ScalarParameter() {defaultValue=0.4, group=SoupGoal_Reactivity, parameterName=ReactivityRampDownExp}
  ) {constExponent=0.4}
  n100 = Subtract(A=Time(), B=ScalarParameter() {defaultValue=-100, group=SoupGoal_Reactivity, parameterName=ReactivityStartTime_4})
  n101 = ScalarParameter() {defaultValue=0.2, group=SoupGoal_Reactivity, parameterName=ReactivityRampUpDuration}
  n102 = Power(
   Base=Saturate(None=Divide(A=@n100, B=@n101) {constB=0.8})
   Exp=ScalarParameter() {defaultValue=0.8, group=SoupGoal_Reactivity, parameterName=ReactivityRampUpExp}
  ) {constExponent=0.8}
  n103 = Divide(A=@n100, B=ScalarParameter() {defaultValue=5, group=SoupGoal_Reactivity, parameterName=ReactivityRampDownDuration}) {constB=0.8}
  n104 = Power(
   Base=OneMinus(None=Saturate(None=@n103))
   Exp=ScalarParameter() {defaultValue=0.4, group=SoupGoal_Reactivity, parameterName=ReactivityRampDownExp}
  ) {constExponent=0.4}
  n105 = Add(
   A=Multiply(A=Multiply(A=@n97, B=@n99), B=ScalarParameter() {defaultValue=0.25, group=SoupGoal_Reactivity, parameterName=ReactivitySingleHitStrength})
   B=Multiply(A=Multiply(A=@n102, B=@n104), B=ScalarParameter() {defaultValue=0.25, group=SoupGoal_Reactivity, parameterName=ReactivitySingleHitStrength})
  )
  n106 = Add(
   A=Multiply(A=Multiply(A=@n80, B=@n82), B=ScalarParameter() {defaultValue=0.25, group=SoupGoal_Reactivity, parameterName=ReactivitySingleHitStrength})
   B=Add(A=Multiply(A=@n86, B=@n87), B=Add(A=@n94, B=@n105))
  )
  n107 = Time()
  n108 = VectorNoise(World Position=Add(A=Multiply(A=@n57.R) {constB=24}, B=Multiply(A=@n107) {constB=13.72}))
  n109 = Subtract(A=@n106, B=Multiply(A=ComponentMask(None=@n108) {r=true}) {constB=0.8})
  n110 = OneMinus(None=LinearInterpolate(
    A=@n106
    B=Max(A=@n109) {constB=0}
    Alpha=ScalarParameter() {defaultValue=1, group=SoupGoal_Reactivity, parameterName=Reactivity_NoiseCackleAmt, sliderMax=1}
   ) {constAlpha=0.1})
  n111 = Multiply(
   A=StaticSwitchParameter(True=Power(Base=@n57.B, Exp=@n75), False=Multiply(A=@n76, B=@n77)) {parameterName=UseAlternateVeinsMask}
   B=Saturate(None=Subtract(A=@n57.G, B=@n110) {constB=0})
  )
  n112 = Add(
   A=@n73.Emissive Color
   B=Multiply(A=Multiply(A=@n74.RGB, B=@n111), B=ScalarParameter() {defaultValue=15, group=SoupGoal_Reactivity, parameterName=ReactiveBrightness})
  )
  n113 = LinearInterpolate(A=@n73.Base Color, Alpha=@n111)
  MaterialFunctionCall(
   Material=StaticSwitchParameter(
    True=SetMaterialAttributes!(MaterialAttributes=@n73.MaterialAttributes, Emissive Color=@n112, Base Color=@n113) {attributeSetTypes [#2, #1]}
    False=@n72
   ) {group=SoupGoal, parameterName=EmableReactivity}
   M=@n48.M
   Pre Skinned Local Position=@n48.M
   Distance Blend=@n48.M
   Fresnel=@n48.M
  ) {materialFunction=&MF_CharacterParent_2023_PostFX}
  MaterialFunctionCall() {materialFunction=&DebugTimeSine}
  NamedRerouteDeclaration(None=Power(
    Base=StaticComponentMaskParameter(None=@n48.Tech Art Mask) {defaultB=true, group=SoupGoal, parameterName=VeinsMaskChannel}
    Exp=ScalarParameter() {defaultValue=2, parameterName=VeinsMaskExp}
   )) {name=VeinsMask, nodeColor=<rgba 0 0.7891 1 1>}
  NamedRerouteDeclaration(None=StaticSwitchParameter(True=NamedRerouteUsage(), False=NamedRerouteDeclaration(None=@n48.Result) {name=TODAdjustmentsIn, nodeColor=<rgba 0.2891 0 1 1>}) {group=SoupGoal, parameterName=EnableEmissiveTODAdjustments}) {name=MA_PreFX, nodeColor=<rgba 1 0 0.8203 1>}
  n114 = GetMaterialAttributes(None=NamedRerouteUsage()) {attributeGetTypes [#2]}
  n115 = MaterialFunctionCall(
   Color_Morning=Constant3Vector() {constant=<rgba 1 0.7726 0.7 0>}
   Color_Day=Constant3Vector() {constant=<rgba 1 1 1 0>}
   Color_Evening=Constant3Vector() {constant=<rgba 1 0.7132 0.7 0>}
   Color_Night=Constant3Vector() {constant=<rgba 0.5 0.55 1 0>}
  ) {materialFunction=&Color_Time_of_Day}
  NamedRerouteDeclaration(None=SetMaterialAttributes!(MaterialAttributes=@n114.MaterialAttributes, Emissive Color=Multiply(A=@n114.Emissive Color, B=@n115.Result)) {attributeSetTypes [#2]}) {name=TODAdjustmentsOut, nodeColor=<rgba 0.0312 1 0 1>}
  NamedRerouteDeclaration(None=SetMaterialAttributes!(
    MaterialAttributes=@n1.MaterialAttributes
    Base Color=LinearInterpolate(A=@n1.Base Color, B=@n46, Alpha=@n2)
    Emissive Color=@n1.MaterialAttributes
   ) {attributeSetTypes [#1, #2]}) {name=MA_WithInteriorTurbulence, nodeColor=<rgba 0 1 0.8594 1>}
  n116 = LocalPosition() {includedOffsets=ExcludeOffsets, localOrigin=InstancePreSkinning}
  n117 = Divide(
   A=NamedRerouteDeclaration(None=VertexInterpolator(VS=@n116.XYZ).PS) {name=PreskinnedLocalPos, nodeColor=<rgba 1 0.4688 0 1>}
   B=ScalarParameter() {defaultValue=120, group=SoupGoal, parameterName=InnerParallax_TileSizeWS}
  )
  NamedRerouteDeclaration(None=ComponentMask(None=@n117) {b=true, r=true}) {name=WorldUV_YPlane, nodeColor=<rgba 0 1 0.9766 1>}
  NamedRerouteDeclaration(None=ComponentMask(None=@n117) {b=true, g=true}) {name=WorldUV_XPlane, nodeColor=<rgba 0 1 0.9766 1>}
  n118 = Constant3Vector() {constant=<rgba 0 1 0 0>}
  NamedRerouteDeclaration(None=Sign(None=DotProduct(A=@n118, B=@n28.PS))) {name=WorldUV_YSign, nodeColor=<rgba 0 1 0.9766 1>}
  n119 = Constant3Vector() {constant=<rgba 1 0 0 0>}
  NamedRerouteDeclaration(None=Sign(None=DotProduct(A=@n119, B=@n28.PS))) {name=WorldUV_XSign, nodeColor=<rgba 0 1 0.9766 1>}
  n120 = DotProduct(A=VertexNormalWS(), B=CameraVectorWS())
  NamedRerouteDeclaration(None=OneMinus(None=Abs(None=@n120)) {desc=parallax offset intensity}) {name=ParallaxAngle0-1, nodeColor=<rgba 1 0 0.3047 1>}
  NamedRerouteDeclaration(None=Multiply(A=Time(), B=ScalarParameter() {defaultValue=1, group=SoupGoal, parameterName=InnerParallax_OVerallSpeed})) {name=TurbulenceTime, nodeColor=<rgba 0 0.625 1 1>}
  Power(Base=@n57.G, Exp=@n60) {desc=TODO: add gradients for the non-accented veins into texture}
 ]
}
