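# Physics asset for the F_MED_SoupGoal female character's Cape face accessory, set up for physical motion of a five-bone dyn_hair chain.
# The hair bones carry capsules of narrowing radius, linked in sequence from the head with limited swing and tight twist.
# Head, upper spine and upper-arm capsules give the body volume around the chain, joined by constraints with free swing and twist.
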PhysicsAsset /BRCosmetics/Characters/Player/Female/Medium/Bodies/F_MED_SoupGoal/Meshes/Parts/F_MED_SoupGoal_Cape_FaceAcc_Physics
bodies {
 dyn_hair_01 [{
   shapeType=Capsule
   length=8.9368
   radius=7.4729
   shapeName=dyn_hair_01_capsule
   center=<5.3596 0.6319 0.9364>
   rotation=<rot -83.8425 150.3841 17.398>
  }]
 dyn_hair_02 [{
   shapeType=Capsule
   length=6.953
   radius=6.9221
   shapeName=dyn_hair_02_capsule
   center=<5.749 -1.7527 0.3656>
   rotation=<rot 2.5339 95.1535 -87.4251>
  }]
 dyn_hair_03 [{
   shapeType=Capsule
   length=6.653
   radius=4.9642
   shapeName=dyn_hair_03_capsule
   center=<6.1538 -0.5995 0.1157>
   rotation=<rot 0.4132 -86.0579 -88.9861>
  }]
 dyn_hair_04 [{
   shapeType=Capsule
   length=5.7496
   radius=4.3067
   shapeName=dyn_hair_04_capsule
   center=<6.354 -0.4272 0.0964>
   rotation=<rot 0.5433 -94.3866 -87.9762>
  }]
 dyn_hair_05 [{
   shapeType=Capsule
   length=8.8257
   radius=2.8218
   shapeName=dyn_hair_05_capsule
   center=<9.3076 -0.4991 0.4974>
   rotation=<rot -0.1769 88.6925 -88.6369>
  }]
 head [{
   shapeType=Capsule
   length=10.9323
   radius=10.6168
   shapeName=head_capsule
   center=<8.0471 -1.154 -0.3086>
   rotation=<rot 43.5375 -124.3751 -25.2304>
  }]
 spine_04 [{shapeType=Capsule, length=41.51, radius=11, shapeName=spine_04_capsule, center=<0 -3.5115 0>, rotation=<rot 0 -90 0>}]
 spine_05 [{shapeType=Capsule, length=35, radius=12.005, shapeName=spine_05_capsule, center=<0 -1.5509 0>, rotation=<rot 0 -90 0>}]
 upperarm_l [{shapeType=Capsule, length=24.5, radius=7.5, shapeName=upperarm_l_capsule, center=<12.364 0 0>, rotation=<rot 90 0 180>}]
 upperarm_r [{shapeType=Capsule, length=24.5, radius=7.5, shapeName=upperarm_r_capsule, center=<-12.364 0 0>, rotation=<rot 90 0 0>}]
}
constraints [
 {bone1Name=dyn_hair_05, bone2Name=dyn_hair_04, swing1LimitDegrees=15, swing2LimitDegrees=15, twistLimitDegrees=2}
 {bone1Name=dyn_hair_02, bone2Name=dyn_hair_01, swing1LimitDegrees=20, swing2LimitDegrees=10, twistLimitDegrees=2}
 {bone1Name=dyn_hair_03, bone2Name=dyn_hair_02, swing1LimitDegrees=15, swing2LimitDegrees=15, twistLimitDegrees=2}
 {bone1Name=dyn_hair_04, bone2Name=dyn_hair_03, swing1LimitDegrees=15, swing2LimitDegrees=15, twistLimitDegrees=2}
 {bone1Name=dyn_hair_01, bone2Name=head, swing1LimitDegrees=20, swing2LimitDegrees=10, twistLimitDegrees=2}
 {
  bone1Name=head
  bone2Name=spine_05
  swing1LimitDegrees=45
  swing1Motion=Free
  swing2LimitDegrees=45
  swing2Motion=Free
  twistLimitDegrees=45
  twistMotion=Free
 }
 {
  bone1Name=spine_05
  bone2Name=spine_04
  swing1LimitDegrees=45
  swing1Motion=Free
  swing2LimitDegrees=45
  swing2Motion=Free
  twistLimitDegrees=45
  twistMotion=Free
 }
 {
  bone1Name=upperarm_l
  bone2Name=spine_05
  swing1LimitDegrees=45
  swing1Motion=Free
  swing2LimitDegrees=45
  swing2Motion=Free
  twistLimitDegrees=45
  twistMotion=Free
 }
 {
  bone1Name=upperarm_r
  bone2Name=spine_05
  swing1LimitDegrees=45
  swing1Motion=Free
  swing2LimitDegrees=45
  swing2Motion=Free
  twistLimitDegrees=45
  twistMotion=Free
 }
]
