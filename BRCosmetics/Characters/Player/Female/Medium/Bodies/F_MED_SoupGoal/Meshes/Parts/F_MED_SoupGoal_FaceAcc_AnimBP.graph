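# Animation Blueprint for the F_MED_SoupGoal face accessory, moving its dyn_hair bones on a pose copied from the body meshes.
# With rigid body enabled, it selects one of three RigidBody simulations for cape, other backbling or none, with bone offsets on the latter two.
# Otherwise, bone constraints weighted between head and upper spine take over, plus additive corrections driven by head forward/back rotation.
# The update event refreshes that rotation and the backbling flags.
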AnimBlueprint /BRCosmetics/Characters/Player/Female/Medium/Bodies/F_MED_SoupGoal/Meshes/Parts/F_MED_SoupGoal_FaceAcc_AnimBP
target_skeleton=&/F_MED_SoupGoal_FaceAcc_Skeleton
graphs [
 AnimGraph {nodes [
   Copy Pose From Mesh(SourceMeshComponent=Get Body() -> (Body:Skeletal Mesh Component Object Reference))
   n1 = RigidBody!(ComponentPose=Local To Component(LocalPose=Use cached pose 'pose cash'()), Alpha, OverrideWorldGravity:Vector="0,0,-920", ExternalForce) -> (Pose) {
    bClampLinearTranslationLimitToRefPose=true
    bOverrideWorldGravity=true
    overridePhysicsAsset=&/F_MED_SoupGoal_FaceAcc_Physics
    simulationSpace=BaseBoneSpace
    alphaBoolBlend {blendOption=Linear}
    baseBoneRef {boneName=spine_02}
    componentAppliedLinearAccClamp=<4000 3000 3000>
    componentLinearAccScale=<0.6 0.25 0.5>
    componentLinearVelScale=<0.6 0.25 0.5>
    overrideWorldGravity {z=-920}
    simSpaceSettings {
     maxAngularAcceleration=1000
     maxAngularVelocity=3
     maxLinearAcceleration=700
     maxLinearVelocity=500
     velocityScaleZ=0
     worldAlpha=0.2
     externalLinearDragV=<0.2 0.2>
    }
   }
   n2 = "Transform (Modify) Bone - Bone: dyn_hair_04"!(ComponentPose=@n1, bAlphaBoolEnabled:bool=Get bIsInFrontEnd(), Translation="0,-0.5,0", Rotation="0,0,5", Scale="1,1,1") -> (Pose) {
    alphaInputType=Bool
    rotationMode=BMM_Additive
    translationMode=BMM_Additive
    alphaBoolBlend {blendOption=Linear}
    boneToModify {boneName=dyn_hair_04}
    rotation {roll=5}
    translation {y=-0.5}
   }
   n3 = "Transform (Modify) Bone - Bone: dyn_hair_base"(ComponentPose=Local To Component(LocalPose=Use cached pose 'pose cash'()), Alpha=1, Translation="0,0,0", Rotation="0,0,-50", Scale="1,1,1") {rotationMode=BMM_Additive, alphaBoolBlend {blendOption=Linear}, boneToModify {boneName=dyn_hair_base}, rotation {roll=-50}}
   n4 = RigidBody(ComponentPose=@n3, OverrideWorldGravity="0,0,-920") {
    bClampLinearTranslationLimitToRefPose=true
    bOverrideWorldGravity=true
    overridePhysicsAsset=&/F_MED_SoupGoal_Backpack_FaceAcc_Physics
    simulationSpace=BaseBoneSpace
    alphaBoolBlend {blendOption=Linear}
    baseBoneRef {boneName=spine_04}
    componentAppliedLinearAccClamp=<3000 3000 3000>
    componentLinearAccScale=<0.3 0.25 0.5>
    componentLinearVelScale=<0.3 0.25 0.5>
    overrideWorldGravity {z=-920}
    simSpaceSettings {maxAngularAcceleration=40, maxAngularVelocity=3, maxLinearAcceleration=40, maxLinearVelocity=50, velocityScaleZ=0, worldAlpha=1}
   }
   n5 = "Transform (Modify) Bone - Bone: dyn_hair_02"(ComponentPose=@n4, Alpha=1, Translation="-0.000,5.008,0.000", Rotation="-0.000,0.000,9.404", Scale="1,1,1") {
    rotationMode=BMM_Additive
    translationMode=BMM_Additive
    alphaBoolBlend {blendOption=Linear}
    boneToModify {boneName=dyn_hair_02}
    rotation {roll=9.404}
    translation {y=5.008}
   }
   n6 = Blend Poses by bool(
    BlendPose_0=Component To Local(ComponentPose=@n5)
    BlendPose_1=Component To Local(ComponentPose=@n2)
    bActiveValue=Get IsWearingBackbling() -> (IsWearingBackbling:bool)
   ) {blendType=Linear}
   n7 = RigidBody(ComponentPose=Local To Component(LocalPose=Use cached pose 'pose cash'()), OverrideWorldGravity="0,0,-920") {
    bClampLinearTranslationLimitToRefPose=true
    bOverrideWorldGravity=true
    overridePhysicsAsset=&/F_MED_SoupGoal_Cape_FaceAcc_Physics
    simulationSpace=BaseBoneSpace
    alphaBoolBlend {blendOption=Linear}
    baseBoneRef {boneName=spine_02}
    componentAppliedLinearAccClamp=<4000 3000 3000>
    componentLinearAccScale=<0.6 0.25 0.5>
    componentLinearVelScale=<0.6 0.25 0.5>
    overrideWorldGravity {z=-920}
    simSpaceSettings {
     maxAngularAcceleration=1000
     maxAngularVelocity=3
     maxLinearAcceleration=700
     maxLinearVelocity=500
     velocityScaleZ=0
     worldAlpha=0.2
     externalLinearDragV=<0.2 0.2>
    }
   }
   n8 = Component To Local(ComponentPose=@n7)
   n9 = Get Is Backbling A Cape() -> (Is Backbling A Cape:bool)
   n10 = "Constraint - dyn_hair_01 (PP)"!(ComponentPose=Local To Component(LocalPose=Use cached pose 'pose cash'()), Alpha=1, ConstraintWeights_0=0.8, ConstraintWeights_1:float=0.2) -> (Pose) {
    alphaBoolBlend {blendOption=Linear}
    boneToModify {boneName=dyn_hair_01}
    constraintSetup [{targetBone {boneName=head}}, {targetBone {boneName=spine_05}}]
    constraintWeights [1,1]
   }
   n11 = "Constraint - dyn_hair_02 (PP)"!(ComponentPose=@n10, Alpha=1, ConstraintWeights_0=0.7, ConstraintWeights_1:float=0.3) -> (Pose) {
    alphaBoolBlend {blendOption=Linear}
    boneToModify {boneName=dyn_hair_02}
    constraintSetup [{targetBone {boneName=head}}, {targetBone {boneName=spine_05}}]
    constraintWeights [1,1]
   }
   n12 = "Constraint - dyn_hair_03 (PP)"!(ComponentPose=@n11, Alpha=1, ConstraintWeights_0=0.4, ConstraintWeights_1:float=0.6) -> (Pose) {
    alphaBoolBlend {blendOption=Linear}
    boneToModify {boneName=dyn_hair_03}
    constraintSetup [{targetBone {boneName=head}}, {targetBone {boneName=spine_05}}]
    constraintWeights [1,1]
   }
   n13 = "Constraint - dyn_hair_04 (PP)"!(ComponentPose=@n12, Alpha=1, ConstraintWeights_0=0.2, ConstraintWeights_1:float=0.8) -> (Pose) {
    alphaBoolBlend {blendOption=Linear}
    boneToModify {boneName=dyn_hair_04}
    constraintSetup [{targetBone {boneName=head}}, {targetBone {boneName=spine_05}}]
    constraintWeights [1,1]
   }
   n14 = "Constraint - dyn_hair_05 (PPP)"!(ComponentPose=@n13, Alpha=1, ConstraintWeights_0=0.2, ConstraintWeights_1:float=0.6, ConstraintWeights_2:float=0.2) -> (Pose) {
    alphaBoolBlend {blendOption=Linear}
    boneToModify {boneName=dyn_hair_05}
    constraintSetup [{targetBone {boneName=head}}, {targetBone {boneName=spine_05}}, {targetBone {boneName=spine_04}}]
    constraintWeights [1,1,1]
   }
   n15 = Get head_fwdBwd_rot() -> (head_fwdBwd_rot:double)
   n16 = "Transform (Modify) Bone - Bone: dyn_hair_02"(ComponentPose=@n14, Alpha=@n15, Translation="0.015,-1,0.678", Rotation="0,-5,0", Scale="1,1,1") {
    rotationMode=BMM_Additive
    rotationSpace=BCS_BoneSpace
    translationMode=BMM_Additive
    translationSpace=BCS_BoneSpace
    alphaBoolBlend {blendOption=Linear}
    alphaScaleBiasClamp {bMapRange=true, inRange {max=0, min=-25}, outRange {max=0, min=1}}
    boneToModify {boneName=dyn_hair_02}
    rotation {yaw=-5}
    translation=<0.015 -1 0.678>
   }
   n17 = "Transform (Modify) Bone - Bone: dyn_hair_03"(ComponentPose=@n16, Alpha=@n15, Translation="-0.001,-0.4,-0.4", Rotation="-0.010,-2.865,0.003", Scale="1,1,1") {
    rotationMode=BMM_Additive
    rotationSpace=BCS_BoneSpace
    translationMode=BMM_Additive
    translationSpace=BCS_BoneSpace
    alphaBoolBlend {blendOption=Linear}
    alphaScaleBiasClamp {bMapRange=true, inRange {max=0, min=-25}, outRange {max=0, min=1}}
    boneToModify {boneName=dyn_hair_03}
    rotation=<rot -0.01 -2.865 0.003>
    translation=<-0.001 -0.4 -0.4>
   }
   n18 = Get head_fwdBwd_rot()
   n19 = "Transform (Modify) Bone - Bone: dyn_hair_02"(ComponentPose=@n17, Alpha=@n18, Translation="0,1,0", Rotation="0,0,0", Scale="1,1,1") {
    rotationMode=BMM_Additive
    rotationSpace=BCS_BoneSpace
    translationMode=BMM_Additive
    translationSpace=BCS_BoneSpace
    alphaBoolBlend {blendOption=Linear}
    alphaScaleBiasClamp {bMapRange=true, inRange {max=15}}
    boneToModify {boneName=dyn_hair_02}
    translation {y=1}
   }
   n20 = "Transform (Modify) Bone - Bone: dyn_hair_03"(ComponentPose=@n19, Alpha=@n18, Translation="0.009,2,0", Rotation="0,0,0", Scale="1,1,1") {
    rotationMode=BMM_Additive
    rotationSpace=BCS_BoneSpace
    translationMode=BMM_Additive
    translationSpace=BCS_BoneSpace
    alphaBoolBlend {blendOption=Linear}
    alphaScaleBiasClamp {bMapRange=true, inRange {max=15}}
    boneToModify {boneName=dyn_hair_03}
    translation=<0.009 2>
   }
   n21 = "Transform (Modify) Bone - Bone: dyn_hair_05"(ComponentPose=@n20, Alpha=@n18, Translation="0,1,0", Rotation="-0.009,11.852,0.002", Scale="1,1,1") {
    rotationMode=BMM_Additive
    rotationSpace=BCS_BoneSpace
    translationMode=BMM_Additive
    translationSpace=BCS_BoneSpace
    alphaBoolBlend {blendOption=Linear}
    alphaScaleBiasClamp {bMapRange=true, inRange {max=15}}
    boneToModify {boneName=dyn_hair_05}
    rotation=<rot -0.009 11.8521 0.002>
    translation {y=1}
   }
   Output Pose(Result=Blend Poses by bool(
     BlendPose_0=Blend Poses by bool(BlendPose_0=@n8, BlendPose_1=@n6, bActiveValue=@n9) {blendType=Linear}
     BlendPose_1=Component To Local(ComponentPose=@n21)
     bActiveValue=Get bRigidBodyEnabled()
    ) {blendType=Linear}) {layerGroup=None}
   n22 = Layered blend per bone(
    BasePose=Copy Pose From Mesh() {bUseAttachedParent=true}
    BlendPoses_0=Copy Pose From Mesh(SourceMeshComponent=Get Body()) {bUseAttachedParent=true}
   ) {layerSetup [{branchFilters [{blendDepth=2, boneName=clavicle_r}, {blendDepth=2, boneName=clavicle_l}, {blendDepth=1, boneName=thigh_l}, {blendDepth=1, boneName=thigh_r}]}]}
   Save cached pose 'pose cash'(Pose=@n22) {cacheName=pose cash}
  ]}
 EventGraph {nodes [
   Event BlueprintUpdateAnimation()
   > n1 = Sequence!(execute) -> (then_0, then_1, then_2:Exec)
   n2 = Gravity Override State Function DEPRECATED(
    execute=@n1.then_1
    CharacterPartAnimInstance=Self-Reference()
    GameTimeInSeconds=GetGameTimeInSeconds()
    GravityOverrideParams=Get Gravity Override Params()
   )
   > Set Gravity Override(Gravity Override=@n2)
   n3 = CheckForBackbling()
   Set IsWearingBackbling(execute=@n1.then_2, IsWearingBackbling:bool=@n3) -> (Output_Get:bool)
   > Set Is Backbling A Cape(Is Backbling A Cape:bool=@n3.bIsBackblingACape) -> (Output_Get:bool)
   n4 = GetDeltaTransformFromRefPose!(self=GetOwnerSkeletalMeshComponent(), BoneName=head, BaseName=spine_05) -> (
    ReturnValue_Location:Vector
    ReturnValue_Rotation_Roll:float
    ReturnValue_Rotation_Pitch:float
    ReturnValue_Rotation_Yaw:float
    ReturnValue_Scale:Vector
   )
   n5 = GetDeltaTransformFromRefPose(self=GetOwnerSkeletalMeshComponent(), BoneName=head, BaseName=spine_05)
   > Set Head bend(Head bend:double=float + float(
     A=MapRangeClamped(Value=@n5.ReturnValue_Rotation_Yaw, InRangeA=10, InRangeB=-40)
     B=MapRangeClamped(Value=@n4.ReturnValue_Rotation_Yaw, InRangeA=-50, InRangeB=-100, OutRangeB=-1)
    )) -> (Output_Get:double)
   Set head_fwdBwd_rot(
    execute=[@n8.then, @n1.then_0]
    head_fwdBwd_rot:double=GetDeltaTransformFromRefPose(self=GetOwnerSkeletalMeshComponent(), BoneName=head, BaseName=neck_01).ReturnValue_Rotation_Yaw
   ) -> (Output_Get:double)
   n6 = Event OnCharacterCustomizationLoaded()
   > Set Body(Body:Skeletal Mesh Component Object Reference=Get Array Item with EFort Custom Part Type(Array=@n6.CustomMeshParts)) -> (Output_Get:Skeletal Mesh Component Object Reference)
   n7 = Get IsWearingBackbling()
   n8 = Branch(Condition="Nearly Equal (Float)"(A="To Float (Boolean)"(InBool=@n7), B=1, ErrorTolerance=0.1))
   Set head_fwdBwd_rot(execute=@n8.else, head_fwdBwd_rot=0)
  ]}
]
variables [
 Body {type=SkeletalMeshComponent}
 Gravity Override {type=Vector}
 Gravity Override Params {
  type=GravityOverrideParamsStruct
  default {skydive_FloatingDown {gravityOverride=<0 0 8000>}, skydive_Surfing {gravityOverride=<500 -300 2000>}}
 }
 Head bend {type=number}
 head_fwdBwd_rot {type=number}
 Is Backbling A Cape {type=boolean}
 IsWearingBackbling {type=boolean}
]
